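; Passive quest ability that reports progress when the player aims out of a car window.
; On activation it waits for the in-vehicle tag, checks the vehicle's ability system tags against the car tag, then waits for the targeting tag.
; Targeting triggers an in-vehicle recheck with empty branches, then a quest stat event is sent and the ability ends.
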(event Ability|EventActivateAbility
  (|CallParentFunction)
  (Ability|Tasks|WaitGameplayTagAdd "(TagName=\"Gameplay.Action.Player.AthenaInVehicle\")" 0 true
    (:Added
      ;[1 1 1] Check that the player is in an appropriate vehicle
      (if (GameplayTags|TagContainer|CompareTagAssetInterfacetoOtherTags (Variables|Default|GetT_Car) (Vehicle|GetVehicleAbilitySystemComponent (Vehicle|GetVehicleInterface (Variables|GAVars|Data|GetOwningPlayerPawn))))
        ;[1 1 1] Wait for player to aim out the window and make sure they are still in a vehicle
        (Ability|Tasks|WaitGameplayTagAdd "(TagName=\"Gameplay.Status.IsTargeting\")" 0 true
          (:Added
            (bind _bdidactorhavetags (Class|AthenaQuestFunctionLibrary|AthenaCheckActorHasTags (Variables|GAVars|Data|GetOwningPlayerPawn) "(GameplayTags=((TagName=\"Gameplay.Action.Player.AthenaInVehicle\")))"))
            (if _bdidactorhavetags
              _
              (else
                _))
            ;[1 1 1] Send quest progress
            (Quest|SendComplexCustomStatEvent (Variables|GAVars|Data|GetOwningPlayerQuestManager) 0 (Variables|Default|GetT_Quest) "(GameplayTags=)" 1)
            (Ability|EndAbility)))))))

(event Ability|EventFailedToActivatePassiveAbility
  (params (ActorInfo :type Gameplay%20Ability%20Actor%20Info%20Structure)))
